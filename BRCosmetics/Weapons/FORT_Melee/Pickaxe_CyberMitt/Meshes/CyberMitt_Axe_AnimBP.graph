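# Animation Blueprint for the CyberMitt pickaxe that adds secondary motion to its two three-bone cords, A and B.
# When anim dynamics is enabled, sprung AnimDynamics chains run over the cached input pose, with a separately tuned set and additive offsets for off-pawn display.
# When disabled, additive bone rotations pose the cords statically.
# The update event sets blend alphas from front-end and on-pawn state, using off-pawn tuning only in the front end.
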AnimBlueprint /BRCosmetics/Weapons/FORT_Melee/Pickaxe_CyberMitt/Meshes/CyberMitt_Axe_AnimBP
target_skeleton=&/CyberMitt_Axe_Skeleton
graphs [
 AnimGraph {nodes [
   n1 = Get Alpha_SimSpace() -> (Alpha_SimSpace:double)
   n2 = Anim Dynamics - Bone: dyn_cordA_01!(
    ComponentPose=Local To Component(LocalPose=Use cached pose 'Base_pose'())
    Alpha=@n1
    GravityOverride:Vector="0,0,-980"
    ExternalForce="0,0,0"
   ) -> (Pose) {
    angularDampingOverride=1
    angularSpringConstant=10000
    bAngularSpring=true
    bOverrideAngularDamping=true
    bUseGravityOverride=true
    gravityScale=0.35
    simulationSpace=RootRelative
    alphaBoolBlend {blendOption=Linear}
    boundBone {boneName=dyn_cordA_01}
    componentLinearAccScale=<1 1 1>
    componentLinearVelScale=<1 1 1>
    gravityOverride {z=-980}
    physicsBodyDefinitions [{
      boundBone {boneName=dyn_cordA_01}
      boxExtents=<20 30 30>
      constraintSetup {angularLimitsMax=<45 45 45>, angularLimitsMin=<-45 -45 -45>, angularTarget=<1 0 0>}
      localJointOffset=<4 0 0>
     }]
    relativeSpaceBone {boneName=Melee_TwoHanded_Handle}
    simSpaceSettings {maxAngularVelocity=100}
   }
   n3 = Anim Dynamics - Bone: dyn_cordA_02!(ComponentPose=@n2, Alpha=@n1, GravityOverride:Vector="0,0,-980", ExternalForce="0,0,0") -> (Pose) {
    angularDampingOverride=1
    angularSpringConstant=10000
    bAngularSpring=true
    bOverrideAngularDamping=true
    bUseGravityOverride=true
    gravityScale=0.35
    simulationSpace=RootRelative
    alphaBoolBlend {blendOption=Linear}
    boundBone {boneName=dyn_cordA_02}
    componentLinearAccScale=<1 1 1>
    componentLinearVelScale=<1 1 1>
    gravityOverride {z=-980}
    physicsBodyDefinitions [{
      boundBone {boneName=dyn_cordA_02}
      boxExtents=<20 30 30>
      constraintSetup {angularLimitsMax=<35 35 35>, angularLimitsMin=<-35 -35 -35>, angularTarget=<1 0 0>}
      localJointOffset=<4 0 0>
     }]
    simSpaceSettings {maxAngularVelocity=100}
   }
   n4 = Anim Dynamics - Bone: dyn_cordA_03!(ComponentPose=@n3, Alpha=@n1, GravityOverride:Vector="0,0,-980", ExternalForce="0,0,0") -> (Pose) {
    angularDampingOverride=1
    angularSpringConstant=10000
    bAngularSpring=true
    bOverrideAngularDamping=true
    bUseGravityOverride=true
    gravityScale=0.35
    simulationSpace=RootRelative
    alphaBoolBlend {blendOption=Linear}
    boundBone {boneName=dyn_cordA_03}
    componentLinearAccScale=<1 1 1>
    componentLinearVelScale=<1 1 1>
    gravityOverride {z=-980}
    physicsBodyDefinitions [{
      boundBone {boneName=dyn_cordA_03}
      boxExtents=<20 30 30>
      constraintSetup {angularLimitsMax=<25 25 25>, angularLimitsMin=<-25 -25 -25>, angularTarget=<1 0 0>}
      localJointOffset=<4 0 0>
     }]
    simSpaceSettings {maxAngularVelocity=100}
   }
   n5 = Anim Dynamics - Bone: dyn_cordB_01!(ComponentPose=@n4, Alpha=@n1, GravityOverride:Vector="0,0,-980", ExternalForce="0,0,0") -> (Pose) {
    angularDampingOverride=1
    angularSpringConstant=10000
    bAngularSpring=true
    bOverrideAngularDamping=true
    bUseGravityOverride=true
    gravityScale=0.35
    simulationSpace=RootRelative
    alphaBoolBlend {blendOption=Linear}
    boundBone {boneName=dyn_cordB_01}
    componentLinearAccScale=<1 1 1>
    componentLinearVelScale=<1 1 1>
    gravityOverride {z=-980}
    physicsBodyDefinitions [{
      boundBone {boneName=dyn_cordB_01}
      boxExtents=<20 30 30>
      constraintSetup {angularLimitsMax=<45 45 45>, angularLimitsMin=<-45 -45 -45>, angularTarget=<1 0 0>}
      localJointOffset=<4 0 0>
     }]
    simSpaceSettings {maxAngularVelocity=100}
   }
   n6 = Anim Dynamics - Bone: dyn_cordB_02!(ComponentPose=@n5, Alpha=@n1, GravityOverride:Vector="0,0,-980", ExternalForce="0,0,0") -> (Pose) {
    angularDampingOverride=1
    angularSpringConstant=10000
    bAngularSpring=true
    bOverrideAngularDamping=true
    bUseGravityOverride=true
    gravityScale=0.35
    simulationSpace=RootRelative
    alphaBoolBlend {blendOption=Linear}
    boundBone {boneName=dyn_cordB_02}
    componentLinearAccScale=<1 1 1>
    componentLinearVelScale=<1 1 1>
    gravityOverride {z=-980}
    physicsBodyDefinitions [{
      boundBone {boneName=dyn_cordB_02}
      boxExtents=<20 30 30>
      constraintSetup {angularLimitsMax=<35 35 35>, angularLimitsMin=<-35 -35 -35>, angularTarget=<1 0 0>}
      localJointOffset=<4 0 0>
     }]
    simSpaceSettings {maxAngularVelocity=100}
   }
   n7 = Anim Dynamics - Bone: dyn_cordB_03!(ComponentPose=@n6, Alpha=@n1, GravityOverride:Vector="0,0,-980", ExternalForce="0,0,0") -> (Pose) {
    angularDampingOverride=1
    angularSpringConstant=10000
    bAngularSpring=true
    bOverrideAngularDamping=true
    bUseGravityOverride=true
    gravityScale=0.35
    simulationSpace=RootRelative
    alphaBoolBlend {blendOption=Linear}
    boundBone {boneName=dyn_cordB_03}
    componentLinearAccScale=<1 1 1>
    componentLinearVelScale=<1 1 1>
    gravityOverride {z=-980}
    physicsBodyDefinitions [{
      boundBone {boneName=dyn_cordB_03}
      boxExtents=<20 30 30>
      constraintSetup {angularLimitsMax=<25 25 25>, angularLimitsMin=<-25 -25 -25>, angularTarget=<1 0 0>}
      localJointOffset=<4 0 0>
     }]
    simSpaceSettings {maxAngularVelocity=100}
   }
   n8 = Get Alpha_IsNotOnPawn() -> (Alpha_IsNotOnPawn:double)
   n9 = Anim Dynamics - Bone: dyn_cordA_01(ComponentPose=@n7, Alpha=@n8, GravityOverride="0,0,-980", ExternalForce="0,0,0") {
    angularDampingOverride=0.9999
    angularSpringConstant=10000
    bAngularSpring=true
    bOverrideAngularDamping=true
    bUseGravityOverride=true
    gravityScale=0.35
    simulationSpace=RootRelative
    alphaBoolBlend {blendOption=Linear}
    boundBone {boneName=dyn_cordA_01}
    componentLinearAccScale=<1 1 1>
    componentLinearVelScale=<1 1 1>
    gravityOverride {z=-980}
    physicsBodyDefinitions [{
      boundBone {boneName=dyn_cordA_01}
      boxExtents=<20 30 30>
      constraintSetup {angularLimitsMax=<45 45 45>, angularLimitsMin=<-45 -45 -45>, angularTarget=<1 0 0>}
      localJointOffset=<4 0 0>
     }]
    relativeSpaceBone {boneName=Melee_TwoHanded_Handle}
    simSpaceSettings {simSpaceAngularAlpha=0.4}
   }
   n10 = Anim Dynamics - Bone: dyn_cordA_02(ComponentPose=@n9, Alpha=@n8, GravityOverride="0,0,-980", ExternalForce="0,0,0") {
    angularDampingOverride=1
    angularSpringConstant=10000
    bAngularSpring=true
    bOverrideAngularDamping=true
    bUseGravityOverride=true
    gravityScale=0.35
    simulationSpace=RootRelative
    alphaBoolBlend {blendOption=Linear}
    boundBone {boneName=dyn_cordA_02}
    componentLinearAccScale=<1 1 1>
    componentLinearVelScale=<1 1 1>
    gravityOverride {z=-980}
    physicsBodyDefinitions [{
      boundBone {boneName=dyn_cordA_02}
      boxExtents=<20 30 30>
      constraintSetup {angularLimitsMax=<35 35 35>, angularLimitsMin=<-35 -35 -35>, angularTarget=<1 0 0>}
      localJointOffset=<4 0 0>
     }]
    simSpaceSettings {simSpaceAngularAlpha=0.6}
   }
   n11 = Anim Dynamics - Bone: dyn_cordA_03(ComponentPose=@n10, Alpha=@n8, GravityOverride="0,0,-980", ExternalForce="0,0,0") {
    angularDampingOverride=1
    angularSpringConstant=10000
    bAngularSpring=true
    bOverrideAngularDamping=true
    bUseGravityOverride=true
    gravityScale=0.35
    simulationSpace=RootRelative
    alphaBoolBlend {blendOption=Linear}
    boundBone {boneName=dyn_cordA_03}
    componentLinearAccScale=<1 1 1>
    componentLinearVelScale=<1 1 1>
    gravityOverride {z=-980}
    physicsBodyDefinitions [{
      boundBone {boneName=dyn_cordA_03}
      boxExtents=<20 30 30>
      constraintSetup {angularLimitsMax=<25 25 25>, angularLimitsMin=<-25 -25 -25>, angularTarget=<1 0 0>}
      localJointOffset=<4 0 0>
     }]
    simSpaceSettings {simSpaceAngularAlpha=0.8}
   }
   n12 = Anim Dynamics - Bone: dyn_cordB_01(ComponentPose=@n11, Alpha=@n8, GravityOverride="0,0,-980", ExternalForce="0,0,0") {
    angularDampingOverride=0.9999
    angularSpringConstant=10000
    bAngularSpring=true
    bOverrideAngularDamping=true
    bUseGravityOverride=true
    gravityScale=0.35
    simulationSpace=RootRelative
    alphaBoolBlend {blendOption=Linear}
    boundBone {boneName=dyn_cordB_01}
    componentLinearAccScale=<1 1 1>
    componentLinearVelScale=<1 1 1>
    gravityOverride {z=-980}
    physicsBodyDefinitions [{
      boundBone {boneName=dyn_cordB_01}
      boxExtents=<20 30 30>
      constraintSetup {angularLimitsMax=<45 45 45>, angularLimitsMin=<-45 -45 -45>, angularTarget=<1 0 0>}
      localJointOffset=<4 0 0>
     }]
    simSpaceSettings {simSpaceAngularAlpha=0.4}
   }
   n13 = Anim Dynamics - Bone: dyn_cordB_02(ComponentPose=@n12, Alpha=@n8, GravityOverride="0,0,-980", ExternalForce="0,0,0") {
    angularDampingOverride=1
    angularSpringConstant=10000
    bAngularSpring=true
    bOverrideAngularDamping=true
    bUseGravityOverride=true
    gravityScale=0.35
    simulationSpace=RootRelative
    alphaBoolBlend {blendOption=Linear}
    boundBone {boneName=dyn_cordB_02}
    componentLinearAccScale=<1 1 1>
    componentLinearVelScale=<1 1 1>
    gravityOverride {z=-980}
    physicsBodyDefinitions [{
      boundBone {boneName=dyn_cordB_02}
      boxExtents=<20 30 30>
      constraintSetup {angularLimitsMax=<35 35 35>, angularLimitsMin=<-35 -35 -35>, angularTarget=<1 0 0>}
      localJointOffset=<4 0 0>
     }]
    simSpaceSettings {simSpaceAngularAlpha=0.6}
   }
   n14 = Anim Dynamics - Bone: dyn_cordB_03(ComponentPose=@n13, Alpha=@n8, GravityOverride="0,0,-980", ExternalForce="0,0,0") {
    angularDampingOverride=1
    angularSpringConstant=10000
    bAngularSpring=true
    bOverrideAngularDamping=true
    bUseGravityOverride=true
    gravityScale=0.35
    simulationSpace=RootRelative
    alphaBoolBlend {blendOption=Linear}
    boundBone {boneName=dyn_cordB_03}
    componentLinearAccScale=<1 1 1>
    componentLinearVelScale=<1 1 1>
    gravityOverride {z=-980}
    physicsBodyDefinitions [{
      boundBone {boneName=dyn_cordB_03}
      boxExtents=<20 30 30>
      constraintSetup {angularLimitsMax=<25 25 25>, angularLimitsMin=<-25 -25 -25>, angularTarget=<1 0 0>}
      localJointOffset=<4 0 0>
     }]
    simSpaceSettings {simSpaceAngularAlpha=0.8}
   }
   n15 = Get Alpha_IsNotOnPawn()
   n16 = "Transform (Modify) Bone - Bone: dyn_cordB_01"(ComponentPose=@n14, Alpha=@n15, Translation="0,0,0", Rotation="-20,-5,0", Scale="1,1,1") {rotationMode=BMM_Additive, alphaBoolBlend {blendOption=Linear}, boneToModify {boneName=dyn_cordB_01}, rotation {pitch=-20, yaw=-5}}
   n17 = "Transform (Modify) Bone - Bone: dyn_cordA_01"(ComponentPose=@n16, Alpha=@n15, Translation="0,0,0", Rotation="-20,-5,0", Scale="1,1,1") {rotationMode=BMM_Additive, alphaBoolBlend {blendOption=Linear}, boneToModify {boneName=dyn_cordA_01}, rotation {pitch=-20, yaw=-5}}
   n18 = "Transform (Modify) Bone - Bone: cordB_master"(ComponentPose=@n17, Alpha=@n15, Translation="0,0,0", Rotation="85,-5,0", Scale="1,1,1") {rotationMode=BMM_Additive, alphaBoolBlend {blendOption=Linear}, boneToModify {boneName=cordB_master}, rotation {pitch=85, yaw=-5}}
   n19 = "Transform (Modify) Bone - Bone: cordA_master"(ComponentPose=@n18, Alpha=@n15, Translation="0,0,0", Rotation="85,-5,0", Scale="1,1,1") {rotationMode=BMM_Additive, alphaBoolBlend {blendOption=Linear}, boneToModify {boneName=cordA_master}, rotation {pitch=85, yaw=-5}}
   n20 = "Transform (Modify) Bone - Bone: dyn_cordB_02"(ComponentPose=@n19, Alpha=@n15, Translation="0,0,0", Rotation="-15,-5,0", Scale="1,1,1") {rotationMode=BMM_Additive, alphaBoolBlend {blendOption=Linear}, boneToModify {boneName=dyn_cordB_02}, rotation {pitch=-15, yaw=-5}}
   n21 = "Transform (Modify) Bone - Bone: dyn_cordA_02"(ComponentPose=@n20, Alpha=@n15, Translation="0,0,0", Rotation="-15,-5,0", Scale="1,1,1") {rotationMode=BMM_Additive, alphaBoolBlend {blendOption=Linear}, boneToModify {boneName=dyn_cordA_02}, rotation {pitch=-15, yaw=-5}}
   n22 = "Transform (Modify) Bone - Bone: dyn_cordB_03"(ComponentPose=@n21, Alpha=@n15, Translation="0,0,0", Rotation="-5,-5,0", Scale="1,1,1") {rotationMode=BMM_Additive, alphaBoolBlend {blendOption=Linear}, boneToModify {boneName=dyn_cordB_03}, rotation {pitch=-5, yaw=-5}}
   n23 = "Transform (Modify) Bone - Bone: dyn_cordA_03"(ComponentPose=@n22, Alpha=@n15, Translation="0,0,0", Rotation="-5,-5,0", Scale="1,1,1") {rotationMode=BMM_Additive, alphaBoolBlend {blendOption=Linear}, boneToModify {boneName=dyn_cordA_03}, rotation {pitch=-5, yaw=-5}}
   n24 = "Transform (Modify) Bone - Bone: dyn_cordB_01"(ComponentPose=Local To Component(LocalPose=Use cached pose 'Base_pose'()), Alpha=1, Translation="0,0,0", Rotation="45,5,0", Scale="1,1,1") {rotationMode=BMM_Additive, alphaBoolBlend {blendOption=Linear}, boneToModify {boneName=dyn_cordB_01}, rotation {pitch=45, yaw=5}}
   n25 = "Transform (Modify) Bone - Bone: dyn_cordB_02"(ComponentPose=@n24, Alpha=1, Translation="0,0,0", Rotation="35,30,0", Scale="1,1,1") {rotationMode=BMM_Additive, alphaBoolBlend {blendOption=Linear}, boneToModify {boneName=dyn_cordB_02}, rotation {pitch=35, yaw=30}}
   n26 = Get Alpha_IsOnPawn() -> (Alpha_IsOnPawn:double)
   n27 = "Transform (Modify) Bone - Bone: dyn_cordB_03"(ComponentPose=@n25, Alpha=@n26, Translation="0,0,0", Rotation="25,15,0", Scale="1,1,1") {rotationMode=BMM_Additive, alphaBoolBlend {blendOption=Linear}, boneToModify {boneName=dyn_cordB_03}, rotation {pitch=25, yaw=15}}
   n28 = "Transform (Modify) Bone - Bone: dyn_cordA_01"(ComponentPose=@n27, Alpha=1, Translation="0,0,0", Rotation="45,10,0", Scale="1,1,1") {rotationMode=BMM_Additive, alphaBoolBlend {blendOption=Linear}, boneToModify {boneName=dyn_cordA_01}, rotation {pitch=45, yaw=10}}
   n29 = "Transform (Modify) Bone - Bone: dyn_cordA_02"(ComponentPose=@n28, Alpha=1, Translation="0,0,0", Rotation="35,30,0", Scale="1,1,1") {rotationMode=BMM_Additive, alphaBoolBlend {blendOption=Linear}, boneToModify {boneName=dyn_cordA_02}, rotation {pitch=35, yaw=30}}
   n30 = "Transform (Modify) Bone - Bone: dyn_cordA_03"(ComponentPose=@n29, Alpha=@n26, Translation="0,0,0", Rotation="25,15,0", Scale="1,1,1") {rotationMode=BMM_Additive, alphaBoolBlend {blendOption=Linear}, boneToModify {boneName=dyn_cordA_03}, rotation {pitch=25, yaw=15}}
   n31 = Get Alpha_IsNotOnPawn()
   n32 = "Transform (Modify) Bone - Bone: cordB_master"(ComponentPose=@n30, Alpha=@n31, Translation="0,0,0", Rotation="90,0,-2", Scale="1,1,1") {rotationMode=BMM_Additive, alphaBoolBlend {blendOption=Linear}, boneToModify {boneName=cordB_master}, rotation {pitch=90, roll=-2}}
   n33 = "Transform (Modify) Bone - Bone: cordA_master"(ComponentPose=@n32, Alpha=@n31, Translation="0,0,0", Rotation="100,0,-2", Scale="1,1,1") {rotationMode=BMM_Additive, alphaBoolBlend {blendOption=Linear}, boneToModify {boneName=cordA_master}, rotation {pitch=100, roll=-2}}
   n34 = "Transform (Modify) Bone - Bone: dyn_cordB_01"(ComponentPose=@n33, Alpha=@n31, Translation="0,0,0", Rotation="-60,0,-2", Scale="1,1,1") {rotationMode=BMM_Additive, alphaBoolBlend {blendOption=Linear}, boneToModify {boneName=dyn_cordB_01}, rotation {pitch=-60, roll=-2}}
   n35 = "Transform (Modify) Bone - Bone: dyn_cordA_01"(ComponentPose=@n34, Alpha=@n31, Translation="0,0,0", Rotation="-60,0,-2", Scale="1,1,1") {rotationMode=BMM_Additive, alphaBoolBlend {blendOption=Linear}, boneToModify {boneName=dyn_cordA_01}, rotation {pitch=-60, roll=-2}}
   n36 = Get Alpha_IsOnPawn()
   n37 = "Transform (Modify) Bone - Bone: cordB_master"(ComponentPose=@n35, Alpha=@n36, Translation="0,0,0", Rotation="-5,0,5", Scale="1,1,1") {rotationMode=BMM_Additive, alphaBoolBlend {blendOption=Linear}, boneToModify {boneName=cordB_master}, rotation {pitch=-5, roll=5}}
   n38 = "Transform (Modify) Bone - Bone: cordA_master"(ComponentPose=@n37, Alpha=@n36, Translation="0,0,0", Rotation="-5,0,5", Scale="1,1,1") {rotationMode=BMM_Additive, alphaBoolBlend {blendOption=Linear}, boneToModify {boneName=cordA_master}, rotation {pitch=-5, roll=5}}
   Output Pose(Result=Blend Poses by bool(
     BlendPose_0=Component To Local(ComponentPose=@n23)
     BlendPose_1=Component To Local(ComponentPose=@n38)
     bActiveValue=Get bAnimDynamicsEnabled() -> (bAnimDynamicsEnabled:bool)
    ) {blendType=Linear}) {layerGroup=None}
   AnimGraphNode_SubInput_0 = Input Pose()
   Save cached pose 'Base_pose'(Pose=@AnimGraphNode_SubInput_0) {cacheName=Base_pose}
  ]}
 EventGraph {nodes [
   Event BlueprintUpdateAnimation()
   > n1 = Branch(Condition=Get bIsInFrontEnd())
   > n2 = Branch(Condition=Get bIsOnPawn() -> (bIsOnPawn:bool))
   > Set Alpha_IsNotOnPawn(Alpha_IsNotOnPawn:double=0) -> (Output_Get:double)
   > Set Alpha_IsOnPawn(Alpha_IsOnPawn:double=1) -> (Output_Get:double)
   > Set Alpha_SimSpace(Alpha_SimSpace:double=1) -> (Output_Get:double)
   Set Alpha_IsNotOnPawn(execute=@n2.else, Alpha_IsNotOnPawn=1)
   > Set Alpha_IsOnPawn(Alpha_IsOnPawn=0)
   > Set Alpha_SimSpace(Alpha_SimSpace=0)
   Set Alpha_IsNotOnPawn(execute=@n1.else, Alpha_IsNotOnPawn=0)
   > Set Alpha_IsOnPawn(Alpha_IsOnPawn=0)
   > Set Alpha_SimSpace(Alpha_SimSpace=1)
   TryGetPawnOwner()
  ]}
]
variables [
 Alpha_IsNotOnPawn {type=number}
 Alpha_IsOnPawn {type=number}
 Alpha_SimSpace {type=number}
 Gravity Override {type=Vector, default=<0 0 -980>}
]
